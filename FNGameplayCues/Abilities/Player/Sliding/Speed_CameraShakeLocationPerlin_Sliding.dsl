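; Sliding camera shake Blueprint that softens its location oscillation as the player's FOV narrows.
; On play, it caches the base X, Y and Z amplitudes and reads a minimum scalar from Fort.MinSlidingCameraShakeScalar, capped at one.
; Each update maps the first local player's FOV, up to the default, onto that scalar-to-one range.
; It rebuilds LocOscillation from the cached amplitudes, avoiding compounding, and returns POV unchanged.
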(event CameraShake|EventReceivePlayShake
  (params (Scale :type Float%20%28single-precision%29))
  (bind _locoscillation (Variables|Oscillation|GetLocOscillation))
  (bind (_x _y _z) (Utilities|Struct|BreakVOscillator _locoscillation))
  (Variables|Default|SetOscillatorAmplitudeX (Utilities|Struct|BreakFOscillator _x))
  (Variables|Default|SetOscillatorAmplitudeY (Utilities|Struct|BreakFOscillator _y))
  (Variables|Default|SetOscillatorAmplitudeZ (Utilities|Struct|BreakFOscillator _z))
  (Variables|Default|SetMinShakeScalar (Math|Float|Min(Float) (Fortnite|GetNumericCVar "Fort.MinSlidingCameraShakeScalar") 1.0)))

(fn BlueprintUpdateCameraShake
  (params (DeltaTime :type Float%20%28single-precision%29) (Alpha :type Float%20%28single-precision%29) (POV :type Minimal%20View%20Info%20Structure%20%28by%20ref%29))
  (bind (_x _y _z) (Utilities|Struct|BreakVOscillator (Variables|Oscillation|GetLocOscillation)))
  (bind (_amplitude _frequency _waveform) (Utilities|Struct|BreakFOscillator _x))
  (bind (_amplitude_2 _frequency_2 _waveform_2) (Utilities|Struct|BreakFOscillator _z))
  (bind (_amplitude_3 _frequency_3 _waveform_3) (Utilities|Struct|BreakFOscillator _y))
  (LocalVariables|SetFOVScale (Math|Float|MapRangeClamped (Camera|GetFOVAngle (Variables|/Script/Engine.PlayerController|GetPlayerCameraManager (Fortnite|GetFirstLocalFortPlayerController true))) 2e+01 (Variables|/Script/Engine.PlayerCameraManager|GetDefaultFOV (Variables|/Script/Engine.PlayerController|GetPlayerCameraManager (Fortnite|GetFirstLocalFortPlayerController true))) (Variables|Default|GetMinShakeScalar) 1.0))
  (Variables|Oscillation|SetLocOscillation (Utilities|Struct|MakeVOscillator (Utilities|Struct|MakeFOscillator (* (Variables|Default|GetOscillatorAmplitudeX) (LocalVariables|GetFOVScale)) _frequency _waveform) (Utilities|Struct|MakeFOscillator (* (Variables|Default|GetOscillatorAmplitudeY) (LocalVariables|GetFOVScale)) _frequency_3 _waveform_3) (Utilities|Struct|MakeFOscillator (* (Variables|Default|GetOscillatorAmplitudeZ) (LocalVariables|GetFOVScale)) _frequency_2 _waveform_2)))
  (return POV))
